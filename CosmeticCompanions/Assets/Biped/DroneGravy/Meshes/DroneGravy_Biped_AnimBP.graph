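# Animation Blueprint for the DroneGravy biped cosmetic companion, posing it by IK-retargeting animation from another mesh at runtime.
# A Blend Poses by bool on bUseReverseRetarget (not declared here) crossfades between two Retarget Pose From Mesh nodes: the emote retargeter when true, the standard retargeter when false.
# A Copy Pose From Mesh node is left unconnected, and the Event Graph holds no logic beyond a lone TryGetPawnOwner.
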AnimBlueprint /CosmeticCompanions/Assets/Biped/DroneGravy/Meshes/DroneGravy_Biped_AnimBP
target_skeleton=&/DroneGravy_Biped_Skeleton
graphs [
 AnimGraph {nodes [
   Copy Pose From Mesh() {bCopyCurves=true, bUseAttachedParent=true}
   Output Pose(Result=Blend Poses by bool(
     BlendPose_0=Retarget Pose From Mesh!() -> (Pose:Pose Link Structure) {
      is=title AnimGraphNode_RetargetPoseFromMesh
      iKRetargeterAsset=&/IK_Retargeter_DroneGravy_Biped_Emote
     }
     BlendPose_1=Retarget Pose From Mesh() {iKRetargeterAsset=&/IK_Retargeter_DroneGravy_Biped}
     BlendTime_0=0.2
     BlendTime_1=0.3
     bActiveValue=Get bUseReverseRetarget() -> (bUseReverseRetarget:bool)
    ))
  ]}
 EventGraph {nodes [TryGetPawnOwner()]}
]
variables []
